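; A burst gameplay cue notify, named as the left-hand splash for a player sprinting in water.
; On burst it casts the target to an Athena player pawn and spawns a Niagara system attached to that pawn's mesh.
; It looks like placeholder work: the effect is a metal impact, it snaps to the RightHand socket despite the left-hand name, and a "HANDSPLASH_LEFT" debug string is printed.
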(fn OnBurst(Cascade)
  (params (MyTarget :type Actor%20Object%20Reference) (Parameters :type Gameplay%20Cue%20Parameters%20Structure%20%28by%20ref%29) (ParticleComponents :type Array%20of%20Cascade%20Particle%20System%20Component%20Object%20References) (AudioComponents :type Array%20of%20Audio%20Component%20Object%20References) (BurstCameraShakeInstance :type Camera%20Shake%20Base%20Object%20Reference) (BurstDecalInstance :type Decal%20Actor%20Object%20Reference))
  (bind _asplayer_pawn_athena (Utilities|Casting|CastToPlayerPawn_Athena MyTarget))
  (Niagara|SpawnSystemAttached "/Game/Weapons/FORT_Generic_Effects/Impacts/Metal/NS_Impact_Metal_Athena/NS_Impact_Metal_Athena.NS_Impact_Metal_Athena" (Variables|/Script/Engine.Character|GetMesh _asplayer_pawn_athena) "RightHand" "0, 0, 0" "0, 0, 0" "SnapToTarget")
  (Development|PrintString 0 "HANDSPLASH_LEFT")
  (return))
